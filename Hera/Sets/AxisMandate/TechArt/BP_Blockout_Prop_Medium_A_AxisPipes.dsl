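; Blockout medium pipe prop for the Axis Mandate set, with per-instance color and flow-animation settings.
; After the parent construction script, it turns the invert-flow-direction flag into an Animation value of 1 or 0.
; It writes hue offset, luminance factor, animation offset and that value to static mesh custom primitive data slots, avoiding dynamic material instances.
; BeginPlay, overlap and Tick events are empty.
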(event EventBeginPlay)

(event Collision|EventActorBeginOverlap
  (params (OtherActor :type Actor%20Object%20Reference)))

(event EventTick
  (params (DeltaSeconds :type Float%20%28single-precision%29)))

(fn ConstructionScript
  (bind _staticmeshcomponent (Variables|Debug|GetStaticMeshComponent))
  (|CallParentFunction)
  (if (Variables|Default|GetInvert%20Flow%20Direction%20%28animation%29)
    (Variables|Default|SetAnimation 1.0)
    (else
      (Variables|Default|SetAnimation 0.0)))
  (Rendering|Material|SetCustomPrimitiveDataFloat _staticmeshcomponent 0 (Variables|Default|GetHue%20Offset))
  (Rendering|Material|SetCustomPrimitiveDataFloat _staticmeshcomponent 1 (Variables|Default|GetLuminance%20Factor))
  (Rendering|Material|SetCustomPrimitiveDataFloat _staticmeshcomponent 2 (Variables|Default|GetAnimation%20Offset))
  (Rendering|Material|SetCustomPrimitiveDataFloat _staticmeshcomponent 3 (Variables|Default|GetAnimation)))
