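# Material function that layers animated, glowing bubbly water for liquid tubes onto incoming material attributes, masked by vertex red behind a UseWaterBubbles switch.
# Three panned, bump-offset bubble layers with distortion, noise and Fresnel masks drive tinted, exposure-compensated emissive.
# UVs optionally follow local height for straight pipes; custom primitive data controls hue, luminance and animation offset or inversion.
# A material proxy replacement omits the emissive for proxy rendering.
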MaterialFunction /Hera/Sets/AxisMandate/TechArt/MF_LiquidTubes_Water
graph {nodes [
  Constant()
  Constant()
  Fresnel() {baseReflectFraction=0, exponent=2}
  n1 = FunctionInput() {inputType=FunctionInput_MaterialAttributes}
  n2 = NamedRerouteUsage()
  n3 = ScalarParameter() {parameterName=Metallic}
  n4 = ScalarParameter() {defaultValue=1, parameterName=Specular}
  n5 = BreakMaterialAttributes!(Attr=NamedRerouteUsage())
  n6 = NamedRerouteUsage()
  n7 = ScalarParameter() {defaultValue=0.8, parameterName=ExposureCompensation}
  n8 = SetMaterialAttributes!(
   MaterialAttributes
   BaseColor=@n2
   Metallic=@n3
   Specular=@n4
   Roughness=@n5.Roughness
   SubsurfaceColor
   EmissiveColor=LinearInterpolate(A=EyeAdaptationInverse(LightValueInput=@n6, AlphaInput=@n7).EyeAdaptationInverse, Alpha=NamedRerouteUsage()) {constB=0}
   CustomData0
  ) {attributeSetTypes [
    #1
    #2
    #3
    #4
    #5
    #6
    #7
   ]}
  n9 = SetMaterialAttributes(BaseColor=@n2, Metallic=@n3, Specular=@n4, Roughness=@n5.Roughness) {attributeSetTypes [
    #1
    #2
    #3
    #4
    #5
    #6
    #7
   ]}
  n10 = MaterialProxyReplace(Realtime=@n8, MaterialProxy=@n9)
  n11 = VertexColor()
  FunctionOutput(None=StaticSwitchParameter(True=BlendMaterialAttributes(A=@n1, B=@n10, Alpha=@n11.R), False=@n1) {defaultValue=true, parameterName=UseWaterBubbles})
  LinearInterpolate(
   A=ScalarParameter() {defaultValue=0.2, parameterName=BGEmissiveMult_Min}
   B=ScalarParameter() {defaultValue=1, parameterName=BGEmissiveMult_Max}
   Alpha=NamedRerouteUsage()
  ) {constA=0.5}
  MaterialFunctionCall!("Normal (WS)", Texture Scale, Triplanar Fade, Derivatives Scale) {materialFunction=/Game/Athena/Environments/MaterialFunctions/MF_TriplanarDithered}
  n12 = SmoothStep(
   Min=ScalarParameter() {defaultValue=0.1447, parameterName=Param_5}
   Max=ScalarParameter() {defaultValue=-0.0371, parameterName=Param_6}
   Value=NamedRerouteUsage()
  )
  Multiply(A=VectorParameter() {parameterName="Color_Water_C(Fresnel)", defaultValue=<rgba 0.1111 0.9312 0.9826 1>}.RGB, B=@n12)
  n13 = NamedRerouteUsage()
  n14 = ScalarParameter() {defaultValue=0.5, parameterName=Noise Distortion}
  n15 = ScalarParameter() {defaultValue=500, parameterName=NoiseTiling}
  n16 = MaterialFunctionCall!(
   WorldPosition=NamedRerouteUsage()
   World Space Normal
   UVSize_Front=@n15
   UVSize_Side=@n15
   UVSize_Top=@n15
   ProjectionTransitionContrast=Constant() {r=0.1}
  ) {materialFunction=/Game/Packages/Fortress_SharedMaterials/Base_Material_Functions/WorldAlignedTexture_UVsMasks}
  n17 = Multiply(A=@n16.UV_Front) {constB=0.4}
  n18 = ComponentMask(None=@n17) {r=true}
  n19 = Frac(None=Multiply(A=Time(), B=ScalarParameter() {defaultValue=0.1, group=Slurp, parameterName=Bubble_SHScaleNoiseSpeed}) {constB=0.05})
  n20 = Add(A=ComponentMask(None=@n17) {g=true}, B=@n19)
  n21 = TextureObjectParameter() {
   group=Slurp
   parameterName=Bubble_SHScaleNoise
   samplerType=SAMPLERTYPE_Grayscale
   texture=&LowResBlurredNoise
  }
  n22 = TextureSample(UVs=Add(A=Multiply(A=@n13, B=@n14) {constB=0.1}, B=AppendVector(A=@n18, B=@n20)), Tex=@n21) {
   constMipValue=0
   samplerSource=SSM_Wrap_WorldGroupSettings
   samplerType=SAMPLERTYPE_Grayscale
   texture=&LowResBlurredNoise
  }
  n23 = Multiply(A=@n16.UV_Side) {constB=0.4}
  n24 = ComponentMask(None=@n23) {r=true}
  n25 = Add(A=ComponentMask(None=@n23) {g=true}, B=@n19)
  n26 = NamedRerouteUsage()
  n27 = TextureSample(UVs=Add(A=AppendVector(A=@n24, B=@n25), B=Multiply(A=@n26, B=@n14) {constB=0.1}), Tex=@n21) {
   constMipValue=0
   samplerSource=SSM_Wrap_WorldGroupSettings
   samplerType=SAMPLERTYPE_Grayscale
   texture=&LowResBlurredNoise
  }
  n28 = LinearInterpolate(A=@n22.R, B=@n27.R)
  Multiply(A=@n28, B=ScalarParameter() {defaultValue=0.5, parameterName=Param_1})
  NamedRerouteDeclaration(None=@n16.Mask_Top) {name=Bubble_UVMask_Top, nodeColor=<rgba 0.1016 0 1 1>}
  n29 = ComponentMask(None=TextureCoordinate()) {g=true}
  n30 = AppendVector(A=Divide(A=LocalPosition().Z, B=ScalarParameter() {defaultValue=-150, parameterName=HeightDivision}), B=@n29)
  n31 = TextureCoordinate()
  NamedRerouteDeclaration(None=Multiply(
    A=ScalarParameter() {defaultValue=1, parameterName=Base Tilling}
    B=StaticSwitchParameter(True=@n30, False=@n31) {parameterName=IsPipeStraight}
   )) {name=BubbleMappingRaw, nodeColor=<rgba 0.9219 1 0 1>}
  n32 = MaterialFunctionCall(Power=ScalarParameter() {defaultValue=0.4887, parameterName=WaterBG_FresnelPow}) {materialFunction=&Fresnel_Function}
  n33 = OneMinus(None=@n32.Result)
  n34 = SmoothStep(
   Min=ScalarParameter() {defaultValue=-0.264, parameterName=Fresnel highlight Min 1}
   Max=ScalarParameter() {defaultValue=1.8037, parameterName=Fresnel highlight Max 1}
   Value=@n33
  )
  n35 = SmoothStep(
   Min=ScalarParameter() {defaultValue=0.4134, parameterName=Fresnel highlight Min 2}
   Max=ScalarParameter() {defaultValue=1.159, parameterName=Fresnel highlight Max 2}
   Value=@n33
  )
  NamedRerouteDeclaration(None=Add(A=Add(A=@n34, B=@n35)) {constB=0.1}) {name=Fresnel highlight, nodeColor=<rgba 0.4062 0 1 1>}
  n36 = ConstantBiasScale(None=NamedRerouteUsage()) {bias=-0.5, scale=2}
  n37 = NamedRerouteUsage()
  n38 = ScalarParameter() {defaultValue=3, parameterName=Bubble_Tiling_X1}
  n39 = ScalarParameter() {defaultValue=2.5, parameterName=Bubble_Tiling_Y1}
  n40 = Multiply(A=NamedRerouteUsage(), B=Multiply(A=AppendVector(A=@n38, B=@n39), B=NamedRerouteUsage()))
  n41 = NamedRerouteUsage()
  n42 = NamedRerouteUsage()
  n43 = Add(
   A=Multiply(A=Multiply(A=@n36, B=@n37), B=ScalarParameter() {defaultValue=0.5, parameterName=Bubble_DistortionAmount1})
   B=BumpOffset(
    Coordinate=Panner(Coordinate=@n40, Time=@n41, Speed=@n42)
    Height=ScalarParameter() {defaultValue=-0.15, parameterName=Bubble_BumpOffsetHeight_1}
    HeightRatioInput=ScalarParameter() {defaultValue=0.05, parameterName=Bubble_BumpOffsetHRI_1}
   )
  )
  n44 = Add(A=@n43, B=ScalarParameter() {defaultValue=0.1, parameterName=Bubble_SwizzleOffset_1})
  n45 = SphereMask(
   A=Fmod(A=@n43, B=ScalarParameter() {defaultValue=2, parameterName=Param_2})
   B=AppendVector(
    A=ScalarParameter() {defaultValue=1.6, parameterName=SwizzleMask_1_X}
    B=ScalarParameter() {defaultValue=1.75, parameterName=SwizzleMask_1_Y}
   )
   Radius=ScalarParameter() {defaultValue=0.5, parameterName=SwizzleMask_1_Radius}
  ) {attenuationRadius=0.3, hardnessPercent=50}
  n46 = TextureSample(
   UVs=NamedRerouteDeclaration(None=LinearInterpolate(A=@n43, B=@n44, Alpha=@n45)) {name=Bubble_UV_1, nodeColor=<rgba 1 0 0.7266 1>}
   Tex=TextureObject() {samplerType=SAMPLERTYPE_LinearColor, texture=/Game/Global/Materials/MaterialGym/Materials/Nitro/T_NitroBubbles}
  ) {samplerSource=SSM_Wrap_WorldGroupSettings}
  n47 = Multiply(A=Subtract(A=MaterialFunctionCall(Float3=@n46.RGB) {materialFunction=&BreakOutFloat3Components}.R) {constB=0.2}) {constA=1}
  n48 = Saturate(None=NamedRerouteDeclaration(None=Saturate(None=@n47)) {name=Bubble_Texture_1, nodeColor=<rgba 0.2188 1 0 1>})
  n49 = Multiply(A=Multiply(A=@n48, B=ScalarParameter() {defaultValue=1, group=Slurp, parameterName=BubbleIntensity_1}) {constB=0.5}, B=NamedRerouteUsage())
  n50 = ScalarParameter() {defaultValue=0.2, parameterName=Bubble_DistortionAmount2}
  n51 = NamedRerouteUsage()
  n52 = ScalarParameter() {defaultValue=1, parameterName=Bubble_Tiling_X2}
  n53 = ScalarParameter() {defaultValue=0.8, parameterName=Bubble_Tiling_Y2}
  n54 = Multiply(A=NamedRerouteUsage(), B=Multiply(A=AppendVector(A=@n52, B=@n53), B=NamedRerouteUsage()))
  n55 = NamedRerouteUsage()
  n56 = NamedRerouteUsage()
  n57 = Add(
   A=Multiply(A=NamedRerouteUsage(), B=Multiply(A=@n50, B=@n51))
   B=BumpOffset(
    Coordinate=Panner(Coordinate=@n54, Time=@n55, Speed=@n56)
    Height=ScalarParameter() {defaultValue=-0.1, parameterName=Bubble_BumpOffsetHeight_2}
    HeightRatioInput=ScalarParameter() {defaultValue=0.05, parameterName=Bubble_BumpOffsetHRI_2}
   )
  )
  n58 = TextureObjectParameter() {
   group=Slurp
   parameterName=BubbleMask
   samplerType=SAMPLERTYPE_Masks
   texture=/Game/Characters/Player/Male/Large/Bodies/M_LRG_Slurp_Monster/Textures/T_SlurpBubble
  }
  n59 = MaterialFunctionCall(Float3=TextureSample(UVs=NamedRerouteDeclaration(None=@n57) {name=Bubble_UV_2, nodeColor=<rgba 1 0 0.7266 1>}, Tex=@n58) {samplerSource=SSM_Wrap_WorldGroupSettings}.G) {materialFunction=&BreakOutFloat3Components}
  n60 = Multiply(A=@n59.R, B=ScalarParameter() {defaultValue=1, group=Slurp, parameterName=BubbleIntensity_2})
  n61 = ScalarParameter() {defaultValue=0.25, parameterName=Bubble_DistortionAmount3}
  n62 = NamedRerouteDeclaration(None=Multiply(A=ScalarParameter() {defaultValue=1, parameterName=GEneral Distortion}, B=NamedRerouteUsage())) {name=Distort, nodeColor=<rgba 1 0 0.8672 1>}
  n63 = ScalarParameter() {defaultValue=3, parameterName=Bubble_Tiling_X3}
  n64 = ScalarParameter() {defaultValue=2.2, parameterName=Bubble_Tiling_Y3}
  n65 = Multiply(A=NamedRerouteUsage(), B=Multiply(A=AppendVector(A=@n63, B=@n64), B=NamedRerouteUsage()))
  n66 = NamedRerouteUsage()
  n67 = NamedRerouteUsage()
  n68 = Add(
   A=Multiply(A=NamedRerouteUsage(), B=Multiply(A=@n61, B=@n62))
   B=BumpOffset(
    Coordinate=Panner(Coordinate=@n65, Time=@n66, Speed=@n67)
    Height=ScalarParameter() {defaultValue=0.2, parameterName=Bubble_BumpOffsetHeight_3}
    HeightRatioInput=ScalarParameter() {defaultValue=0.1, parameterName=Bubble_BumpOffsetHRI_3}
   )
  )
  n69 = TextureObjectParameter() {
   group=Slurp
   parameterName=BubbleMask
   samplerType=SAMPLERTYPE_Masks
   texture=&T_SlurpBubble
  }
  n70 = MaterialFunctionCall(Float3=TextureSample(UVs=NamedRerouteDeclaration(None=@n68) {name=Bubble_UV_3, nodeColor=<rgba 1 0 0.7266 1>}, Tex=@n69) {samplerSource=SSM_Wrap_WorldGroupSettings}.G) {materialFunction=&BreakOutFloat3Components}
  n71 = NamedRerouteDeclaration(None=Add(A=@n70.R, B=@n70.R)) {name=Bubble 3, nodeColor=<rgba 0.2188 1 0 1>}
  n72 = ScalarParameter() {defaultValue=1, group=Slurp, parameterName=BubbleIntensity_3}
  NamedRerouteDeclaration(None=Add(A=Add(A=@n49, B=@n60), B=Multiply(A=@n71, B=@n72))) {name=Msk_Bubbles, nodeColor=<rgba 1 0 0.6328 1>}
  n73 = PreSkinnedPosition()
  NamedRerouteDeclaration(None=ShaderStageSwitch(PixelShader=VertexInterpolator(VS=@n73).PS, VertexShader=@n73)) {name=PreSkinnedPos_VS-PS, nodeColor=<rgba 0 0.4844 1 1>}
  n74 = VectorParameter() {
   bUseCustomPrimitiveData=true
   parameterName=LiquidAdjustments
   channelNames {a="Invert Animation (0-1)", b=Animation Offset, g=Luminance Factor, r="Hue Offset (0-1)"}
   defaultValue=<rgba 0 1 0 1>
  }
  NamedRerouteDeclaration(None=@n74.Animation Offset) {name=Custom Offset, nodeColor=<rgba 0 0.8359 1 1>}
  NamedRerouteDeclaration(None=Power(Base=Saturate(None=@n28), Exp=ScalarParameter() {defaultValue=2, parameterName=NoisePower}) {constExponent=4}) {name=Bubble_Noise, nodeColor=<rgba 0 0.4609 1 1>}
  n75 = OneMinus(None=MaterialFunctionCall() {materialFunction=&Fresnel_Function}.Result)
  n76 = Add(A=@n48, B=NamedRerouteDeclaration(None=Add(A=@n59.R, B=@n59.R)) {name=Bubble_Texture_2, nodeColor=<rgba 0.2188 1 0 1>})
  NamedRerouteDeclaration(None=Saturate(None=Multiply(A=@n75, B=@n76))) {name=Bubble_PostFresnel, nodeColor=<rgba 0 1 0.3203 1>}
  n77 = ComponentMask(None=VectorParameter() {parameterName=NDotL_CustomVector, defaultValue=<rgba 0.5 0.5 1 1>}.RGB) {b=true, g=true, r=true}
  n78 = NamedRerouteUsage()
  n79 = Multiply(A=DotProduct(A=@n77, B=@n78)) {constB=2}
  n80 = Normalize(VectorInput=Subtract(A=Multiply(A=@n79, B=@n78), B=@n77))
  n81 = MaterialFunctionCall() {materialFunction=&CameraDirectionVector}
  n82 = Abs(None=DotProduct(A=@n80, B=Multiply(A=@n81.Result)))
  n83 = Add(A=ComponentMask(None=@n80) {r=true}, B=@n82)
  n84 = Multiply(
   A=Power(Base=Max(A=@n83) {constB=0}, Exp=ScalarParameter() {defaultValue=1, group=Slurp, parameterName=Bubble_BackSpecular_Power})
   B=ScalarParameter() {defaultValue=0.5, group=Slurp, parameterName=Bubble_BackSpecular_Scale}
  ) {constB=0.15}
  n85 = OneMinus(None=@n82)
  n86 = ComponentMask(None=@n80) {b=true}
  n87 = Multiply(
   A=Power(Base=Add(A=@n85, B=@n86), Exp=ScalarParameter() {defaultValue=1, group=Slurp, parameterName=Bubble_HighlightSpecular_Power})
   B=ScalarParameter() {defaultValue=0.3, group=Slurp, parameterName=Bubble_HighlightSpecular_Scale}
  ) {constB=0.55}
  n88 = Add(
   A=Multiply(A=VectorParameter() {group=Slurp, parameterName=Bubble_BackSpecular_Color, defaultValue=<rgba 0 1 1 1>}.RGB, B=Saturate(None=@n84))
   B=Multiply(
    A=Saturate(None=@n87)
    B=VectorParameter() {group=Slurp, parameterName=Bubble_HighlightSpecular_Color, defaultValue=<rgba 0.8281 0.8555 1 1>}.RGB
   )
  )
  n89 = NamedRerouteUsage()
  NamedRerouteDeclaration(None=Multiply(A=Multiply(A=@n88, B=@n89), B=NamedRerouteUsage())) {name=Bubble_Highlight, nodeColor=<rgba 1 0.1875 0 1>}
  n90 = Multiply(A=TextureCoordinate(), B=ScalarParameter() {defaultValue=1.5, parameterName=Param_9})
  n91 = Multiply(
   A=MaterialFunctionCall!(ReflectionVector, ReflectionTexture) {materialFunction=/Engine/Functions/Engine_MaterialFunctions01/Reflections/ViewAlignedReflection}.UVs
   B=ScalarParameter() {defaultValue=0.5, group=Slurp, parameterName=Bubble_ViewDependantDistortion}
  )
  n92 = NamedRerouteUsage()
  n93 = ScalarParameter() {defaultValue=-0.05, group=Slurp, parameterName=Bubble_ViewDistortion_SpeedX}
  n94 = ScalarParameter() {group=Slurp, parameterName=Bubble_ViewDistortion_SpeedY}
  n95 = Rotator(Coordinate=Panner(Coordinate=LinearInterpolate(A=@n90, B=@n91) {constAlpha=0.2}, Time=@n92, Speed=AppendVector(A=@n93, B=@n94)), Time=@n92) {speed=0.0001}
  n96 = TextureSample!(UVs=@n95, Tex, "DDX(UVs)"=DDX(Value=@n95), "DDY(UVs)"=DDY(Value=@n95), Apply View MipBias) {
   constMipValue=0
   mipValueMode=TMVM_Derivative
   samplerSource=SSM_Wrap_WorldGroupSettings
   samplerType=SAMPLERTYPE_Grayscale
   texture=&LowResBlurredNoise
  }
  n97 = Multiply(A=Multiply(A=@n96.G, B=NamedRerouteUsage()), B=NamedRerouteUsage())
  n98 = OneMinus(None=Fresnel(ExponentIn=ScalarParameter() {defaultValue=1, parameterName=Bubble_Fresnel_Power}) {baseReflectFraction=0, exponent=1})
  NamedRerouteDeclaration(None=Power(Base=Multiply(A=@n97, B=@n98)) {constExponent=0.5}) {name=Bubble_GradientMask, nodeColor=<rgba 0.6406 1 0 1>}
  n99 = Time()
  n100 = NamedRerouteUsage()
  n101 = NamedRerouteUsage()
  NamedRerouteDeclaration(None=Multiply(A=Add(A=@n99, B=@n100), B=LinearInterpolate(Alpha=@n101) {constA=1, constB=-1})) {name=Offseted Time, nodeColor=<rgba 1 0.0399 0 1>}
  n102 = AtmosphericLightVector()
  n103 = MaterialFunctionCall(Float3=Constant3Vector() {constant=<rgba 0 0 1 0>}) {materialFunction=&BreakOutFloat3Components}
  n104 = ObjectBounds()
  n105 = Add(A=@n103.B, B=ComponentMask(None=Multiply(A=@n104) {constB=0.5}) {b=true})
  n106 = Subtract(
   A=TransformPosition(None=MaterialFunctionCall(X=@n103.R, Y=@n103.R, Z=@n105) {materialFunction=&MakeFloat3}.Result) {transformType=TRANSFORMPOSSOURCE_World}
   B=WorldPosition().XYZ
  )
  n107 = Add(A=Multiply(A=Normalize(VectorInput=@n106)) {constB=-1})
  n108 = Saturate(None=ConstantBiasScale(None=DotProduct(A=@n102, B=@n107)))
  n109 = ScalarParameter() {defaultValue=1, parameterName=SamplePos}
  n110 = ScalarParameter() {defaultValue=1, parameterName=TransmittanceRadius}
  n111 = ScalarParameter() {parameterName=TransmittanceHardness}
  NamedRerouteDeclaration(None=Power(
    Base=SphereMask(A=@n108, B=@n109, Radius=@n110, Hardness=@n111)
    Exp=ScalarParameter() {defaultValue=1.5, parameterName=TransmittanceFalloff}
   )) {name=TransmittanceMask, nodeColor=<rgba 1 0 0.6328 1>}
  NamedRerouteDeclaration(None=Multiply(A=@n96.G, B=ScalarParameter() {defaultValue=1, parameterName=Streak Noise Mult})) {name=StreakNoise, nodeColor=<rgba 0.125 0 1 1>}
  n112 = Frac(None=Subtract(A=@n74.Luminance Factor) {constB=0})
  n113 = VectorParameter() {parameterName=Color_Water, defaultValue=<rgba 0 0.1594 0.6087 1>}
  n114 = Multiply(A=@n113.RGB, B=Multiply(A=NamedRerouteUsage()) {constB=4.5})
  n115 = SkyLightEnvMapSample(Direction=AtmosphericLightVector(), Roughness=ScalarParameter() {defaultValue=1, parameterName=EnvSampleRoughness})
  n116 = VectorParameter() {parameterName="Color_Water_B(Streak)", defaultValue=<rgba 0 0.1239 1 1>}
  n117 = NamedRerouteUsage()
  n118 = Saturate(None=Multiply(A=NamedRerouteUsage(), B=ScalarParameter() {defaultValue=5, parameterName=Noise Hide}))
  n119 = Multiply(A=NamedRerouteUsage(), B=@n118)
  n120 = Multiply(
   A=LinearInterpolate(
    A=VectorParameter() {parameterName=Color_Bubbles, defaultValue=<rgba 0 0.2235 0.8542 1>}.RGB
    B=VectorParameter() {parameterName=Color_Bubbles_B, defaultValue=<rgba 0.0208 0.4578 0.8438 1>}.RGB
    Alpha=NamedRerouteUsage()
   )
   B=NamedRerouteUsage()
  )
  n121 = LinearInterpolate(
   A=LinearInterpolate(A=Add(A=@n113.RGB, B=@n114), B=LinearInterpolate(A=@n115, B=@n116.RGB) {constAlpha=0.9}, Alpha=Power(Base=@n117))
   B=@n120
   Alpha=@n119
  )
  n122 = Multiply(A=NamedRerouteUsage(), B=@n118)
  n123 = NamedRerouteDeclaration(None=Multiply(A=Add(A=@n119, B=@n122), B=@n120)) {name=Highlights, nodeColor=<rgba 1 0 0.2812 1>}
  n124 = NamedRerouteUsage()
  n125 = ScalarParameter() {defaultValue=-1.7014}
  n126 = MaterialFunctionCall(Hue Shift Percentage=@n74."Hue Offset (0-1)", Texture=Desaturation(None=Add(A=@n121, B=@n123), Fraction=Multiply(A=@n124, B=@n125))) {materialFunction=&HueShift}
  NamedRerouteDeclaration(None=Multiply(A=NamedRerouteUsage(), B=Multiply(A=@n112, B=@n126.Result))) {name=Emissive_OUT, nodeColor=<rgba 0 0.9766 1 1>}
  NamedRerouteDeclaration(None=SphereMask(A=@n29, B=Constant() {r=0.5}) {attenuationRadius=0.499, hardnessPercent=60}) {name=SeamMask, nodeColor=<rgba 1 0.4688 0 1>}
  NamedRerouteDeclaration(None=@n45) {name=SwizzleMask_1, nodeColor=<rgba 1 0 0.6094 1>}
  n127 = ScalarParameter() {defaultValue=0.25, group=Slurp, parameterName=Bubble Speed X_1}
  n128 = ScalarParameter() {group=Slurp, parameterName=Bubble Speed Y_1}
  n129 = ScalarParameter() {defaultValue=2, group=Slurp, parameterName=BubbleSpeedGlobal}
  NamedRerouteDeclaration(None=Multiply(A=AppendVector(A=@n127, B=@n128), B=@n129)) {name=BubbleSpeed_1, nodeColor=<rgba 1 0 0.5625 1>}
  n130 = ScalarParameter() {defaultValue=0.25, group=Slurp, parameterName=Bubble Speed X_2}
  n131 = ScalarParameter() {group=Slurp, parameterName=Bubble Speed Y_2}
  NamedRerouteDeclaration(None=Multiply(A=AppendVector(A=@n130, B=@n131), B=@n129)) {name=BubbleSpeed_2, nodeColor=<rgba 1 0 0.5625 1>}
  n132 = TextureObjectParameter() {
   group=Slurp
   parameterName=BubbleNormal
   samplerType=SAMPLERTYPE_Normal
   texture=/Game/Characters/Player/Male/Large/Bodies/M_LRG_Slurp_Monster/Textures/T_SlurpBubble_N
  }
  n133 = TextureSample!(UVs=NamedRerouteUsage(), Tex=@n132, Apply View MipBias) {samplerSource=SSM_Wrap_WorldGroupSettings, samplerType=SAMPLERTYPE_Normal}
  n134 = Multiply(
   A=MaterialFunctionCall(X=@n133.R, Z=@n133.R) {materialFunction=&MakeFloat3}.Result
   B=NamedRerouteUsage()
  )
  NamedRerouteDeclaration(None=@n134) {name=BubbleWSNormal, nodeColor=<rgba 1 0.9844 0 1>}
  n135 = ScalarParameter() {defaultValue=0.3, group=Slurp, parameterName=Bubble Speed X_3}
  n136 = ScalarParameter() {group=Slurp, parameterName=Bubble Speed Y_3}
  NamedRerouteDeclaration(None=Multiply(A=AppendVector(A=@n135, B=@n136), B=@n129)) {name=BubbleSpeed_3, nodeColor=<rgba 1 0 0.5625 1>}
  NamedRerouteDeclaration(None=MaterialFunctionCall() {materialFunction=&ObjectScale}.Scale Z) {name=ObjectScale, nodeColor=<rgba 0.3359 1 0 1>}
  NamedRerouteDeclaration(None=@n74."Invert Animation (0-1)") {name=Invert Animation, nodeColor=<rgba 1 0.4219 0 1>}
  NamedRerouteDeclaration(None=@n1) {name=MA og, nodeColor=<rgba 0 1 0.1328 1>}
  n137 = PreSkinnedNormal()
  NamedRerouteDeclaration(None=ShaderStageSwitch(PixelShader=VertexInterpolator(VS=@n137).PS, VertexShader=@n137)) {name=PreSkinnedNrm_VS-PS, nodeColor=<rgba 0 0.4844 1 1>}
  n138 = MaterialFunctionCall(Exponent=ScalarParameter() {defaultValue=2.7505, parameterName=Tubes Outer Fresnel}) {materialFunction=&AxisAlignedFresnel}
  NamedRerouteDeclaration(None=Multiply(A=OneMinus(None=@n138.Fresnel), B=@n32.Result)) {name=Fresnel ax, nodeColor=<rgba 1 0.9609 0 1>}
  NamedRerouteUsage()
  NamedRerouteUsage()
  NamedRerouteUsage()
  NamedRerouteUsage()
  NamedRerouteUsage()
  NamedRerouteUsage()
  NamedRerouteUsage()
  NamedRerouteUsage()
  NamedRerouteUsage()
  NamedRerouteUsage()
  NamedRerouteUsage()
  NamedRerouteUsage()
  NamedRerouteUsage()
  Saturate(None=@n12)
  TextureObjectParameter() {
   group=Slurp
   parameterName=BubbleNormal
   samplerType=SAMPLERTYPE_Normal
   texture=&T_SlurpBubble_N
  }
  TextureObjectParameter() {
   group=Slurp
   parameterName=BubbleMask
   samplerType=SAMPLERTYPE_Masks
   texture=&T_SlurpBubble
  }
  TextureObject() {samplerType=SAMPLERTYPE_Normal, texture=/Game/Global/Materials/MaterialGym/Materials/Nitro/T_NitroBubbles_N}
  TextureSample() {samplerType=SAMPLERTYPE_LinearColor, texture=&T_NitroBubbles}
  TextureSample() {samplerType=SAMPLERTYPE_Normal, texture=&T_NitroBubbles_N}
  Time()
  Transform(None=@n134)
  VectorParameter() {parameterName="Color_Water_C(Fresnel)", defaultValue=<rgba 0.1111 0.9312 0.9826 1>}
 ]}
