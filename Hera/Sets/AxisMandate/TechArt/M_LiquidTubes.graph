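# M_LiquidTubes is a Default Lit, Nanite-enabled material for liquid tubes in the AxisMandate set.
# A tiled surface-noise mask is scaled, offset and overlay-blended into roughness, alongside scalar Metallic and Specular.
# These material attributes pass through MF_LiquidTubes_Water, and a Temporal Responsiveness output is set to 1.
# A GigaMix function and subsurface color and opacity parameters sit unconnected.
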Material /Hera/Sets/AxisMandate/TechArt/M_LiquidTubes
graph {
 properties {
  bAutomaticallySetUsageInEditor=false
  bUseMaterialAttributes=true
  bUsedWithNanite=true
  previewMesh=/NarrowFleaJade/Assets/Meshes/NarrowFlea_Gem_DS
  usedShadingModels=Default Lit
 }
 nodes [
  n1 = TextureSample(UVs=Multiply(A=TextureCoordinate(), B=ScalarParameter() {defaultValue=5, parameterName=Tile Rough})) {samplerType=SAMPLERTYPE_Masks, texture=&/T_Surface_Noise}
  n2 = Multiply(A=Add(A=Multiply(A=@n1.R) {constB=0.2}, B=Add(A=@n1.G, B=@n1.G)), B=ScalarParameter() {defaultValue=-0.0463, parameterName=Roughness Mu})
  n3 = Add(A=@n2, B=ScalarParameter() {defaultValue=0.6392, parameterName=Base Rough}) {constB=0.05}
  MaterialFunctionCall!(In=MakeMaterialAttributes!(
    BaseColor
    Metallic=ScalarParameter() {parameterName=Metallic}
    Specular=ScalarParameter() {defaultValue=0.3821, parameterName=Specular}
    Roughness=MaterialFunctionCall(Base=@n3, Blend=@n2) {materialFunction=&Blend_Overlay}.Result
    Anisotropy
    EmissiveColor
    Opacity
    OpacityMask
    Normal
    Tangent
    WorldPositionOffset
    SubsurfaceColor
    ClearCoat
    ClearCoatRoughness
    AmbientOcclusion
    Refraction
    CustomizedUVs_0
    CustomizedUVs_1
    CustomizedUVs_2
    CustomizedUVs_3
    CustomizedUVs_4
    CustomizedUVs_5
    CustomizedUVs_6
    CustomizedUVs_7
    PixelDepthOffset
    ShadingModel
    Displacement
   )) {materialFunction=&/MF_LiquidTubes_Water}
  MaterialFunctionCall() {materialFunction=/Game/Global/Base4Layers_2022/MaterialFunctions/MF_GigaMix_2023}
  ScalarParameter() {defaultValue=0.5, parameterName=SSOpacity}
  TemporalResponsivenessOutput!(None=Constant() {r=1})
  VectorParameter() {parameterName=SubsurfaceColor, defaultValue=<rgba 0 0.0573 0.2188 1>}
 ]
}
